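# Translucent, two-sided unlit material, apparently for the placed frame of the Creative Sky Dome device, with a pale emissive color parameter and an opacity parameter.
# Pixel depth plus a sorting bias is compared with CustomDepth, giving a mask that is one where the biased depth is nearer and zero otherwise.
# That mask multiplies opacity to fake depth sorting against CustomDepth objects, while a shading path switch skips the test on mobile.
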Material /Game/Creative/Devices/SkyDome/Materials/M_CP_Device_SkyDome_Frame_Placed
graph {
 properties {blendMode=BLEND_Translucent, pixelDepthOffsetMode=PDOM_Legacy, shadingModel=MSM_Unlit, twoSided=true, usedShadingModels=Unlit}
 nodes [
  n1 = Add(A=PixelDepth().R, B=ScalarParameter() {defaultValue=-0.25, parameterName=BiasSorting_Far})
  n2 = ComponentMask(None=SceneTexture() {sceneTextureId=PPI_CustomDepth}.Color) {r=true}
  n3 = Constant()
  n4 = Constant() {r=1}
  Multiply(
   A=ScalarParameter() {defaultValue=0.35, parameterName=Opacity}
   B=ShadingPathSwitch(Default=If(A=@n1, B=@n2, "A > B"=@n3, "A == B"=@n3, "A < B"=@n4), Mobile=Constant() {r=1}) {desc=Migrated from FeatureLevelSwitch}
  )
  VectorParameter() {parameterName=Emissive Color, defaultValue=<rgba 0.8737 0.8898 1 1>}
 ]
}
