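# Translucent, unlit, two-sided material for the placed volume of the Creative SkyDome device.
# Color comes from a rotated, panning texture mapped by reflection vector and distorted by a panning water-foam normal map, with optional desaturation and additive overlay tint.
# Opacity blends background, block and backface terms using two-sided sign and a Fresnel weighted by vertex normal Z, fading out on downward-facing surfaces.
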Material /Game/Creative/Devices/SkyDome/Materials/M_CP_Device_SkyDome_Volume_Placed
graph {
 properties {blendMode=BLEND_Translucent, pixelDepthOffsetMode=PDOM_Legacy, shadingModel=MSM_Unlit, twoSided=true, usedShadingModels=Unlit}
 nodes [
  n1 = ScalarParameter() {defaultValue=0.025, group=Opacity, parameterName=BG Color Opacity, sortPriority=3}
  n2 = ScalarParameter() {defaultValue=0.05, group=Opacity, parameterName=Blocks Opacity, sortPriority=10}
  n3 = LinearInterpolate(
   A=Multiply(A=ScalarParameter() {defaultValue=0.1, group=Opacity, parameterName=Backface Opacity, sortPriority=11})
   Alpha=Saturate(None=TwoSidedSign())
  )
  n4 = OneMinus(None=ComponentMask(None=VertexNormalWS()) {b=true})
  n5 = MaterialFunctionCall(
   Brightness=ScalarParameter() {defaultValue=0.325, desc=.1, parameterName=FresnelBrightness}
   Exponent=ScalarParameter() {defaultValue=4, desc=4, parameterName=FresnelExponent}
  ) {materialFunction=&MF_Fresnel_FX}
  n6 = Multiply(A=ComponentMask(None=VertexNormalWS()) {b=true}) {constB=-1}
  Multiply(
   A=LinearInterpolate(A=LinearInterpolate(A=@n1, B=@n2, Alpha=@n3), Alpha=Multiply(A=@n4, B=@n5.Fresnel Inverted Double Sided))
   B=OneMinus(None=Saturate(None=@n6))
  )
  n7 = ComponentMask(None=ReflectionVectorWS()) {g=true, r=true}
  n8 = ScalarParameter() {desc=Rotation in degrees, group=Main Texture, parameterName=Texture Rotation}
  n9 = Constant() {r=0.0028}
  n10 = Panner(
   Coordinate=MaterialFunctionCall(UVs=@n7, Rotation Center=Constant2Vector() {g=0.5, r=0.5}, "Rotation Angle (0-1)"=Multiply(A=@n8, B=@n9)) {materialFunction=&CustomRotator}.Rotated Values
   Speed=AppendVector(
    A=ScalarParameter() {defaultValue=-0.05, group=Main Texture, parameterName=Texture Panning Speed X}
    B=ScalarParameter() {defaultValue=0.125, group=Main Texture, parameterName=Texture Panning Speed Y}
   )
  ) {speedX=-0.1}
  n11 = ScalarParameter() {defaultValue=2, group=Main Texture, parameterName=Texture Tiling}
  n12 = AppendVector(
   A=ScalarParameter() {defaultValue=0.025, desc=0.05, group=Normal Map Distortion, parameterName=Normal Map Panning Speed X}
   B=ScalarParameter() {defaultValue=-0.0063, desc=-0.0125, group=Normal Map Distortion, parameterName=Normal Map Panning Speed Y}
  )
  n13 = TextureSampleParameter2D(UVs=Multiply(
    A=Panner(Coordinate=@n7, Speed=@n12) {desc=movement, speedX=0.05, speedY=-0.0125}
    B=ScalarParameter() {defaultValue=0.5, group=Normal Map Distortion, parameterName=Normal Map Scale}
   ) {constB=0.5, desc=scale}) {
   group=Normal Map Distortion
   parameterName=Normal Map for Distortion
   samplerType=SAMPLERTYPE_Normal
   texture=/FortWater/Textures/02_WaterFoam_Veroni_base_N
  }
  n14 = Multiply(A=@n13.RGB, B=ScalarParameter() {defaultValue=0.025, group=Normal Map Distortion, parameterName=Normal Map Intensity}) {constB=0.025, desc=intensity}
  n15 = TextureSampleParameter2D(UVs=Add(A=Multiply(A=@n10, B=@n11), B=ComponentMask(None=@n14) {g=true, r=true})) {
   group=Main Texture
   parameterName=Main Texture
   texture=/Game/Creative/Devices/SkyDome/Materials/Textures/T_CP_Device_SkyDome_Volume_Placed_01_D
  }
  n16 = Multiply(
   A=StaticSwitchParameter(True=Desaturation(None=@n15.RGB), False=@n15.RGB) {group=Main Texture, parameterName=Desaturate}
   B=ScalarParameter() {defaultValue=0.25, group=Main Texture, parameterName=Main Texture Intensity}
  )
  n17 = VectorParameter() {group=Main Texture, parameterName=Overlay Color, defaultValue=<rgba 0 0 1 1>}
  n18 = ScalarParameter() {defaultValue=0.35, group=Main Texture, parameterName=Overlay Color Intensity}
  StaticSwitchParameter(True=Add(A=@n16, B=Multiply(A=@n17.RGB, B=@n18)), False=@n16) {group=Main Texture, parameterName=Overlay}
 ]
}
